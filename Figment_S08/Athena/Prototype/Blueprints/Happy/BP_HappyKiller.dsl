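; Prototype event actor that takes dropped pickups as sacrifices on overlap and rewards the dropping player with shields.
; The server converts item count through a divisor into a shield gameplay effect stack, fires OnResourceReceived, then destroys the pickup.
; A delayed multicast flies a fire emitter from the pickup to a hard-coded world point via timeline lerp.
; LocalImpact runs on remotes; it, GlobalImpact and Finale play staged sound, particles and camera shake.
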(event Custom|EdgeServerAddResources
  (params (TotalResources :type Integer64))
  ;[1 1 1] Leaf server only
  (if (> TotalResources 0)
    (Development|PrintString "updated resources! " true true "(R=0.000000,G=0.660000,B=1.000000,A=1.000000)" 1e+01)
    (Networking|ForceNetUpdate)))

(event Collision|EventActorBeginOverlap
  (params (OtherActor :type Actor%20Object%20Reference))
  (CallFunction|ItemSacrifice OtherActor))

(event Custom|Multicast_OnItemSacrifice
  (params (InItemToSacrifice :type Fort%20Pickup%20Athena%20Object%20Reference))
  (bind _itemtosacrifice (Variables|Default|GetItemToSacrifice))
  (bind _returnvalue (Transformation|GetActorLocation _itemtosacrifice))
  ;[1 1 1] The Hacky FX Stuff to move Fire
  (Utilities|IsValid InItemToSacrifice
    (:Is%20Valid
      (bind _output_get (Variables|Default|SetItemToSacrifice InItemToSacrifice))
      (Rendering|SetActorHiddenInGame (Class|FortPickup|GetPickupEffectBlueprint _output_get) true)
      (bind _returnvalue_2 (Effects|SpawnEmitteratLocation "/Figment_S08/Athena/Prototype/Blueprints/Happy/FX/P_Happy_OnObject.P_Happy_OnObject" _returnvalue (Math|Rotator|FindLookatRotation (Transformation|GetWorldLocation (Variables|Default|GetSphere)) (Transformation|GetActorLocation (Variables|Default|GetItemToSacrifice)))))
      (Audio|SpawnSoundAttached "/Figment_S08/Athena/Sounds/Events/Dopey/Happy/Happy_Receive_Cue.Happy_Receive_Cue" (Variables|/Script/Engine.Actor|GetRootComponent (Variables|Default|GetItemToSacrifice)))
      (bind (_direction _locationx) (Timeline|Timeline_0 PlayFromStart)
        (:Update
          (Utilities|IsValid _returnvalue_2
            (:Is%20Valid
              (Utilities|IsValid (Variables|Default|GetItemToSacrifice)
                (:Is%20Valid
                  (Transformation|SetWorldLocation _returnvalue_2 (Math|Vector|MakeVector (Math|Float|Lerp (.x _returnvalue) 83952.0 _locationx) (Math|Float|Lerp (.y _returnvalue) 36496.0 _locationx) (Math|Float|Lerp (.z _returnvalue) 7132.0 _locationx)) false false))))))))))

(event Custom|ItemSacrifice
  (params (Item :type Actor%20Object%20Reference))
  (bind _itemtosacrifice (Variables|Default|GetItemToSacrifice))
  (bind _returnvalue (Pickup|GetItemEntry _itemtosacrifice 0))
  (bind (_itemdefinition _itemguid _count) (Class|FortItemFunctionLibrary|BreakFortItemEntry _returnvalue))
  (bind _itemtosacrifice_2 (Variables|Default|GetItemToSacrifice))
  (bind _returnvalue_2 (Pickup|GetPawnWhoDroppedPickup _itemtosacrifice_2))
  (bind _returnvalue_3 (Ability|GetAbilitySystemComponent _returnvalue_2))
  (bind _itemtosacrifice_3 (Variables|Default|GetItemToSacrifice))
  (bind _asfort_pickup_athena (Utilities|Casting|CastToFortPickupAthena Item))
  (Variables|Default|SetItemToSacrifice _asfort_pickup_athena)
  (Utilities|Casting|CastToFortAmmoItemDefinition _itemdefinition
    (:then
      _)
    (:CastFailed
      (Utilities|Casting|CastToFortResourceItemDefinition _itemdefinition)))
  (Utilities|IsValid (Pickup|GetPawnWhoDroppedPickup _itemtosacrifice_3)
    (:Is%20Valid
      (Utilities|FlowControl|Sequence
        (:then_0
          (Utilities|FlowControl|Delay (Variables|Default|GetSetFireDelay))
          (CallFunction|MulticastonItemSacrifice (Variables|Default|GetItemToSacrifice)))
        (:then_1
          (Collision|SetActorEnableCollision _itemtosacrifice_3 false)
          (Development|PrintString (Utilities|String|Append (Utilities|GetDisplayName _itemtosacrifice_3) " IS WHAT THIS IS!") false false)
          (bind _output_get (Variables|Default|SetResourcesSacrificed _count))
          (Development|PrintString (Utilities|String|Append (Utilities|GetObjectName _itemdefinition) (Utilities|String|ToString(Integer) _output_get)) false false)
          (Networking|SwitchHasAuthority
            (:Authority
              (Utilities|FlowControl|Sequence
                (:then_0
                  (Utilities|IsValid _returnvalue_2
                    (:Is%20Valid
                      (if (> (Variables|Default|GetGrantedShieldDivisor) 0)
                        (Variables|Default|SetShieldToGrant (Math|Conversions|ToFloat(Integer) (/ (Variables|Default|GetResourcesSacrificed) (Variables|Default|GetGrantedShieldDivisor))))
                        (bind _returnvalue_4 (Ability|GameplayEffect|SetStackCount (GameplayEffects|MakeOutgoingSpec :self _returnvalue_3 :GameplayEffectClass "/Figment_S08/Athena/Items/Consumables/Shields/GE_Athena_Shields.GE_Athena_Shields_C" :Level 1.0) (Math|Float|Truncate (Variables|Default|GetShieldToGrant))))
                        (GameplayEffects|ApplyGameplayEffectSpecToTarget _returnvalue_3 _returnvalue_4 _returnvalue_3)))))
                (:then_1
                  (Default|CallOnResourceReceived (Math|Conversions|ToInteger64(Integer) (Variables|Default|GetResourcesSacrificed)))
                  (Utilities|FlowControl|Delay 2.0)
                  (Actor|DestroyActor (Variables|Default|GetItemToSacrifice)))))
            (:Remote
              (CallFunction|LocalImpact))))))))

(event Custom|Finale
  (Utilities|FlowControl|Sequence
    (:then_0
      (Audio|PlaySoundatLocation "/Figment_S08/Athena/Sounds/Events/Dopey/Happy/Happy_Event_3_Cue.Happy_Event_3_Cue" (Transformation|GetWorldLocation (Variables|Default|GetP_Happy_Finale))))
    (:then_1
      (Utilities|FlowControl|Delay (Variables|Default|GetDelayBeforeImpactLocalCameraShake))
      (Camera|PlayWorldCameraShake "/Figment_S08/Effects/Fort_Effects/BluePrints/CameraShakes/B_Happy_Shake02.B_Happy_Shake02_C" (Transformation|GetActorLocation) 0.0 (Variables|Default|GetShakeOutRadius)))
    (:then_2
      (Utilities|FlowControl|Delay (Variables|Default|GetDelayBeforeImpactLocal))
      (Components|Activation|Activate (Variables|Default|GetP_Happy_Finale) true))))

(event Custom|SetGrantedShieldDivisor
  (params (GrantedShieldDivisor :type Integer))
  (Variables|Default|SetGrantedShieldDivisor GrantedShieldDivisor))

(event Custom|LocalImpact
  (Utilities|FlowControl|Sequence
    (:then_0
      (Audio|PlaySoundatLocation "/Figment_S08/Athena/Sounds/Events/Dopey/Happy/Happy_Event_1_Cue.Happy_Event_1_Cue" (Transformation|GetWorldLocation (Variables|Default|GetP_Happy_Impact_Local))))
    (:then_1
      (Utilities|FlowControl|Delay (Variables|Default|GetDelayBeforeImpactLocalCameraShake))
      (Camera|PlayWorldCameraShake "/Figment_S08/Effects/Fort_Effects/BluePrints/CameraShakes/B_Happy_Shake01.B_Happy_Shake01_C" (Transformation|GetActorLocation) 0.0 (Variables|Default|GetShakeOutRadius)))
    (:then_2
      (Utilities|FlowControl|Delay (Variables|Default|GetDelayBeforeImpactLocal))
      (Components|Activation|Activate (Variables|Default|GetP_Happy_Impact_Local) true))))

(event Custom|GlobalImpact
  ;[1 1 1] Billy requested this
  (Components|Activation|Activate (Variables|Default|GetHappy_Impact) false)
  (Development|PrintString "global impact hit")
  (Audio|PlaySoundatLocation "/Figment_S08/Athena/Sounds/Events/Dopey/Happy/Happy_Event_2_Cue.Happy_Event_2_Cue" (Transformation|GetWorldLocation (Variables|Default|GetP_Happy_Impact_Local)))
  (Utilities|FlowControl|Delay (Variables|Default|GetDelayBeforeImpactLocal))
  (Camera|PlayWorldCameraShake "/Figment_S08/Effects/Fort_Effects/BluePrints/CameraShakes/B_Happy_Shake01.B_Happy_Shake01_C" (Transformation|GetActorLocation) 0.0 (Variables|Default|GetGlobalCameraShake)))

(fn ConstructionScript)
