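; BP_Happy_CountComponent is a prototype Blueprint component that relays a resource count in the Happy prototype.
; Its ServerCountResources custom event takes an Integer64 CountResources value and broadcasts it through the OnCountResources event dispatcher.
; BeginPlay and Tick are present but empty, and the component does no counting itself.
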(event Custom|ServerCountResources
  (params (CountResources :type Integer64))
  (Default|CallOnCountResources CountResources))

(event EventBeginPlay)

(event EventTick
  (params (DeltaSeconds :type Float%20%28single-precision%29)))
