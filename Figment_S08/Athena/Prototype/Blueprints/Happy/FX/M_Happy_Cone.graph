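# Translucent, unlit, two-sided material for mesh particles, named as the cone effect for the Happy prototype, that renders a soft tinted shape.
# Particle color is multiplied by morning, day, evening and night tints, then fed into Color_Time_of_Day.
# The opacity mask combines particle alpha, axis-aligned fresnel and powered V-coordinate gradients that fade both ends.
# Near-camera fade and depth fade soften it close to the view and at geometry intersections.
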Material /Figment_S08/Athena/Prototype/Blueprints/Happy/FX/M_Happy_Cone
graph {
 properties {
  bUsedWithMeshParticles=true
  blendMode=BLEND_Translucent
  pixelDepthOffsetMode=PDOM_Legacy
  shadingModel=MSM_Unlit
  twoSided=true
  usedShadingModels=Unlit
 }
 nodes [
  n1 = ParticleColor()
  MaterialFunctionCall(
   Color_Morning=Multiply(A=@n1.RGB, B=VectorParameter() {parameterName="Color_Morning (V3)", defaultValue=<rgba 1 1 1 0>}.RGB)
   Color_Day=Multiply(A=@n1.RGB, B=VectorParameter() {parameterName="Color_Day (V3)", defaultValue=<rgba 1 1 1 0>}.RGB)
   Color_Evening=Multiply(A=@n1.RGB, B=VectorParameter() {parameterName="Color_Evening (V3)", defaultValue=<rgba 1.25 1.25 1.25 0>}.RGB)
   Color_Night=Multiply(A=@n1.RGB, B=VectorParameter() {parameterName="Color_Night (V3)", defaultValue=<rgba 0.5 0.5 0.5 0>}.RGB)
  ) {materialFunction=&Color_Time_of_Day}
  n2 = MaterialFunctionCall(Exponent=ScalarParameter() {defaultValue=5, parameterName="Exponent (S)"}) {materialFunction=&AxisAlignedFresnel}
  n3 = ComponentMask(None=TextureCoordinate()) {g=true}
  n4 = OneMinus(None=@n3)
  n5 = Multiply(A=@n1.A, B=Multiply(A=Power(Base=@n3), B=Power(Base=@n4)))
  n6 = ScalarParameter() {defaultValue=500, parameterName=NearFadeDIstance}
  Multiply(
   A=Multiply(
    A=Multiply(A=@n2.Fresnel, B=@n5)
    B=MaterialFunctionCall(Distance=@n6) {materialFunction=&Near_Camera_Fade}.Result
   )
   B=DepthFade(FadeDistance=ScalarParameter() {defaultValue=2500, parameterName=DepthFadeDIstance})
  )
 ]
}
