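# Translucent unlit light pillar material for particles that grows with camera distance.
# SplineThicken sets width in world position offset, and a distance alpha widens it and scales vertices away from the object pivot.
# A static switch measures from BeaconLocation instead, inverting that alpha and also applying it to opacity.
# Panning smoke and rainbow textures feed HDR, time-of-day tinted color, while gradient masks, depth fade and near-camera fade shape opacity.
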Material /Figment_S08/Athena/Prototype/Blueprints/Happy/FX/M_Happy_Light_Pillar_Distance_Scale
graph {
 properties {
  bUsedWithMeshParticles=true
  bUsedWithParticleSprites=true
  blendMode=BLEND_Translucent
  pixelDepthOffsetMode=PDOM_Legacy
  shadingModel=MSM_Unlit
  twoSided=true
  usedShadingModels=Unlit
 }
 nodes [
  n1 = ScalarParameter() {defaultValue=20, parameterName=MaxWidthMult}
  n2 = ObjectPositionWS()
  n3 = Distance(
   A=StaticSwitchParameter(True=VectorParameter() {parameterName=BeaconLocation}.RGB, False=@n2) {parameterName=Use Controlled Distance Vectors for Scaling}
   B=CameraPositionWS()
  )
  n4 = ScalarParameter() {defaultValue=2000, parameterName=DistanceOffset}
  n5 = Clamp(None=Divide(A=Subtract(A=@n3, B=@n4), B=ScalarParameter() {defaultValue=30000, parameterName=ScaleUpDistance}))
  n6 = StaticSwitchParameter(True=OneMinus(None=@n5), False=@n5) {parameterName=Use Controlled Distance Vectors for Scaling}
  n7 = Multiply(
   A=StaticSwitchParameter(
    True=DynamicParameter() {defaultValue=<rgba 100 1 1 1>, paramNames [SplineWidth,Param2,Param3,Param4]}.SplineWidth
    False=ScalarParameter() {defaultValue=1, parameterName=SplineWidth}
   ) {defaultValue=true, parameterName=UseWParticles}
   B=StaticSwitchParameter(True=LinearInterpolate(B=@n1, Alpha=@n6) {constA=1, constB=3}, False=Constant() {r=1}) {defaultValue=true, parameterName=UseScaleUpDistance}
  )
  n8 = Multiply(A=Multiply(A=ScalarParameter() {defaultValue=60, parameterName=MaxScaleIncrease}) {constB=-1}, B=Subtract(A=@n2, B=WorldPosition().XYZ)) {constB=-1}
  Add(
   A=MaterialFunctionCall(WidthBase=@n7, WidthTip=@n7) {materialFunction=&SplineThicken}.WorldPositionOffset
   B=StaticSwitchParameter(True=LinearInterpolate(B=@n8, Alpha=@n6) {constB=0}, False=Constant()) {defaultValue=true, parameterName=UseScaleUpDistance}
  )
  n9 = ParticleColor()
  n10 = VectorParameter() {parameterName=NoiseTiling, defaultValue=<rgba 1 5 0 0>}
  n11 = Panner(
   Coordinate=Multiply(A=TextureCoordinate(), B=ComponentMask(None=@n10.RGB) {g=true, r=true})
   Speed=AppendVector(A=ScalarParameter() {defaultValue=-0.05, parameterName=SpeedX}, B=ScalarParameter() {defaultValue=-0.25, parameterName=SpeedY})
  ) {speedX=-0.05, speedY=-0.25}
  n12 = Power(Base=TextureSample(UVs=TextureCoordinate() {coordinateIndex=1}) {samplerType=SAMPLERTYPE_LinearColor, texture=/Game/Weapons/WeaponSkins/Textures/T_VertGradient}.R)
  n13 = TextureCoordinate() {uTiling=1.5, vTiling=1.5}
  n14 = Panner(Coordinate=LinearInterpolate(A=Multiply(A=@n12) {constB=-0.5}, B=Multiply(A=@n12) {constB=0.5}, Alpha=ComponentMask(None=@n13) {r=true})) {speedX=0.8, speedY=0.8}
  n15 = Add(
   A=Multiply(
    A=TextureSample(UVs=@n11) {texture=/Game/Effects/Fort_Effects/Textures/Tile/T_Inky_Smoke_Tile_Chromatic}.RGB
    B=ScalarParameter() {defaultValue=1, parameterName=NoiseOpacity}
   )
   B=LinearInterpolate(A=TextureSample(UVs=@n14) {texture=/Game/Effects/Fort_Effects/Textures/Grad/T_Gradient_Rainbow_NoRed}.RGB, Alpha=@n12)
  )
  n16 = ScalarParameter() {defaultValue=-1, parameterName=Desaturation}
  n17 = MaterialFunctionCall(
   LDR Input=Clamp(None=@n9.RGB)
   HDR Tint=Multiply(A=Desaturation(None=@n15, Fraction=@n16), B=@n9.RGB)
   HDR Intensity Multiplier=ScalarParameter() {defaultValue=1, parameterName=HDRIntensity}
  ) {materialFunction=&DeriveHDRfromLDR}
  MaterialFunctionCall(
   Color_Morning=Multiply(A=@n17.RemappedRange, B=VectorParameter() {parameterName="Color_Morning (V3)", defaultValue=<rgba 1 1 1 0>}.RGB)
   Color_Day=Multiply(A=@n17.RemappedRange, B=VectorParameter() {parameterName="Color_Day (V3)", defaultValue=<rgba 1 1 1 0>}.RGB)
   Color_Evening=Multiply(A=@n17.RemappedRange, B=VectorParameter() {parameterName="Color_Evening (V3)", defaultValue=<rgba 1.25 1.25 1.25 0>}.RGB)
   Color_Night=Multiply(A=@n17.RemappedRange, B=VectorParameter() {parameterName="Color_Night (V3)", defaultValue=<rgba 0.5 0.5 0.5 0>}.RGB)
  ) {materialFunction=&Color_Time_of_Day}
  n18 = TextureCoordinate()
  n19 = Multiply(
   A=@n12
   B=TextureSample(UVs=Power(Base=@n18) {constExponent=0.6}) {texture=/Figment_S08/Effects/Fort_Effects/Textures/Masks/T_Vert_Gradient_Soft}.R
  )
  n20 = DepthFade(Opacity=@n19, FadeDistance=ScalarParameter() {defaultValue=300, parameterName=DepthFadeDistance}) {fadeDistanceDefault=200}
  n21 = Multiply(
   A=Multiply(A=@n9.A, B=FeatureLevelSwitch(Default=@n20, ES2=@n19, ES3_1=@n19, SM4_REMOVED=@n20, SM5=@n20, SM6=@n20))
   B=MaterialFunctionCall(Distance=ScalarParameter() {defaultValue=200, parameterName=NearFadeDistance}) {materialFunction=&Near_Camera_Fade}.Result
  )
  StaticSwitchParameter(True=LinearInterpolate(B=@n21, Alpha=@n6), False=@n21) {parameterName=Use Controlled Distance Vectors for Scaling}
 ]
}
